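# Animation Blueprint for the ColdTrophy back bling sprite pack, playing idle and reactive animations on the accessory.
# A state machine switches from random idles to chest-or-elimination, damage and knocked-down states, using input events and bIsDBNO.
# Additive pack-bone offsets on the cached pose follow the owner's spine yaw.
# Frontend display schedules delayed reactions; gimbal variables are computed but unused by the AnimGraph.
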AnimBlueprint /BRCosmetics/Accessories/FORT_Backpacks/Backpack_M_MED_ColdTrophy/Sprites/M_MED_ColdTrophy_Sprite_Pack_AnimBP
target_skeleton=&/M_MED_ColdTrophy_Sprite_Pack_Skeleton
graphs [
 AnimGraph {nodes [
   "Transform (Modify) Bone - Bone: root"(Translation="0,0,0", Rotation="0,0,0") {
    rotationMode=BMM_Additive
    rotationSpace=BCS_BoneSpace
    alphaBoolBlend {blendOption=Linear}
    alphaScaleBiasClamp {bInterpResult=true}
    boneToModify {boneName=root}
   }
   n1 = "Transform (Modify) Bone - Bone: pack"(
    ComponentPose=Local To Component(LocalPose=Use cached pose 'Base_Pose'())
    Translation=Get BP Translate() -> (BP Translate:Vector)
    Rotation=Get BP Rotate() -> (BP Rotate:Rotator)
   ) {
    rotationMode=BMM_Additive
    rotationSpace=BCS_BoneSpace
    translationMode=BMM_Additive
    translationSpace=BCS_BoneSpace
    alphaBoolBlend {blendOption=Linear}
    alphaScaleBiasClamp {bInterpResult=true}
    boneToModify {boneName=pack}
    rotation {roll=10}
   }
   Output Pose(Result=Component To Local(ComponentPose=@n1)) {layerGroup=None}
   AnimGraphNode_StateMachine_0 = Idle and Reactive() {is=title AnimGraphNode_StateMachine}
   Save cached pose 'Base_Pose'(Pose=@AnimGraphNode_StateMachine_0) {cacheName=Base_Pose}
  ]}
 Idle and Reactive {
  entry=Idle
  path=AnimGraph.AnimGraphNode_StateMachine_0.Idle and Reactive
  ids=id
  nodes [
   AnimStateNode_0 = Idle()
   AnimStateNode_1 = Reactive() {is=title AnimStateNode}
   AnimStateNode_2 = DBNO() {is=title AnimStateNode}
   AnimStateNode_3 = Damage() {is=title AnimStateNode}
   AnimStateTransitionNode_0 = AnimStateTransitionNode() {bAllowInertializationForSelfTransitions=true}
   AnimStateTransitionNode_1 = AnimStateTransitionNode() {bAllowInertializationForSelfTransitions=true, bAutomaticRuleBasedOnSequencePlayerInState=true}
   AnimStateTransitionNode_2 = AnimStateTransitionNode() {bAllowInertializationForSelfTransitions=true}
   AnimStateTransitionNode_3 = AnimStateTransitionNode() {bAllowInertializationForSelfTransitions=true, bAutomaticRuleBasedOnSequencePlayerInState=true}
   AnimStateTransitionNode_4 = AnimStateTransitionNode() {bAllowInertializationForSelfTransitions=true}
   AnimStateTransitionNode_5 = AnimStateTransitionNode() {bAllowInertializationForSelfTransitions=true}
  ]
  links [
   {
    dst=Idle
    rule=AnimGraph.AnimGraphNode_StateMachine_0.Idle and Reactive.AnimStateTransitionNode_5.Transition
    src=DBNO
    via=AnimStateTransitionNode_5
   }
   {
    dst=Idle
    rule=AnimGraph.AnimGraphNode_StateMachine_0.Idle and Reactive.AnimStateTransitionNode_3.Transition
    src=Damage
    via=AnimStateTransitionNode_3
   }
   {
    dst=DBNO
    rule=AnimGraph.AnimGraphNode_StateMachine_0.Idle and Reactive.AnimStateTransitionNode_4.Transition
    src=Idle
    via=AnimStateTransitionNode_4
   }
   {
    dst=Damage
    rule=AnimGraph.AnimGraphNode_StateMachine_0.Idle and Reactive.AnimStateTransitionNode_2.Transition
    src=Idle
    via=AnimStateTransitionNode_2
   }
   {
    dst=Reactive
    rule=AnimGraph.AnimGraphNode_StateMachine_0.Idle and Reactive.AnimStateTransitionNode_0.Transition
    src=Idle
    via=AnimStateTransitionNode_0
   }
   {
    dst=Idle
    rule=AnimGraph.AnimGraphNode_StateMachine_0.Idle and Reactive.AnimStateTransitionNode_1.Transition
    src=Reactive
    via=AnimStateTransitionNode_1
   }
  ]
 }
 Idle {
  path=AnimGraph.AnimGraphNode_StateMachine_0.Idle and Reactive.AnimStateNode_0.Idle
  nodes [
   n1 = Random Sequence Player!() -> (Pose:Pose Link Structure) {
    is=title AnimGraphNode_RandomPlayer
    entries [
     {
      chanceToPlay=0.75
      maxLoopCount=0
      maxPlayRate=1
      minLoopCount=0
      minPlayRate=1
      sequence=/BRCosmetics/Accessories/FORT_Backpacks/Backpack_M_MED_ColdTrophy/Anim/ColdTrophy_Idle
      blendIn {blendOption=Linear, blendTime=0, customCurve=null}
     }
     {
      chanceToPlay=0.25
      maxLoopCount=1
      maxPlayRate=1
      minLoopCount=0
      minPlayRate=1
      sequence=/BRCosmetics/Accessories/FORT_Backpacks/Backpack_M_MED_ColdTrophy/Anim/ColdTrophy_IdleBreak
      blendIn {blendOption=Linear, blendTime=0, customCurve=null}
     }
    ]
   }
   Output Animation Pose(Result=@n1)
  ]
 }
 Reactive {
  path=AnimGraph.AnimGraphNode_StateMachine_0.Idle and Reactive.AnimStateNode_1.Reactive
  nodes [
   n1 = Random Sequence Player() {entries [
     {
      chanceToPlay=0.5
      maxLoopCount=1
      maxPlayRate=1
      minLoopCount=0
      minPlayRate=1
      sequence=/BRCosmetics/Accessories/FORT_Backpacks/Backpack_M_MED_ColdTrophy/Anim/ColdTrophy_Reactive_Elim
      blendIn {blendOption=Linear, blendTime=0, customCurve=null}
     }
     {
      chanceToPlay=0.5
      maxLoopCount=1
      maxPlayRate=1
      minLoopCount=0
      minPlayRate=1
      sequence=/BRCosmetics/Accessories/FORT_Backpacks/Backpack_M_MED_ColdTrophy/Anim/ColdTrophy_Reactive_ElimAlt
      blendIn {blendOption=Linear, blendTime=0, customCurve=null}
     }
    ]}
   Output Animation Pose(Result=@n1)
  ]
 }
 DBNO {
  path=AnimGraph.AnimGraphNode_StateMachine_0.Idle and Reactive.AnimStateNode_2.DBNO
  nodes [Output Animation Pose(Result=Sequence Player 'ColdTrophy_Reactive_DBNO'() {sequence=/BRCosmetics/Accessories/FORT_Backpacks/Backpack_M_MED_ColdTrophy/Anim/ColdTrophy_Reactive_DBNO})]
 }
 Damage {
  path=AnimGraph.AnimGraphNode_StateMachine_0.Idle and Reactive.AnimStateNode_3.Damage
  nodes [Output Animation Pose(Result=Sequence Player 'ColdTrophy_Reactive_Damage'() {bLoopAnimation=false, sequence=/BRCosmetics/Accessories/FORT_Backpacks/Backpack_M_MED_ColdTrophy/Anim/ColdTrophy_Reactive_Damage})]
 }
 Transition {
  path=AnimGraph.AnimGraphNode_StateMachine_0.Idle and Reactive.AnimStateTransitionNode_0.Transition
  nodes [Result(bCanEnterTransition=Get PlayReactive() -> (PlayReactive:bool))]
 }
 Transition {path=AnimGraph.AnimGraphNode_StateMachine_0.Idle and Reactive.AnimStateTransitionNode_1.Transition, nodes [Result()]}
 Transition {
  path=AnimGraph.AnimGraphNode_StateMachine_0.Idle and Reactive.AnimStateTransitionNode_2.Transition
  nodes [Result(bCanEnterTransition=Get PlayDamageReactive() -> (PlayDamageReactive:bool))]
 }
 Transition {path=AnimGraph.AnimGraphNode_StateMachine_0.Idle and Reactive.AnimStateTransitionNode_3.Transition, nodes [Result()]}
 Transition {
  path=AnimGraph.AnimGraphNode_StateMachine_0.Idle and Reactive.AnimStateTransitionNode_4.Transition
  nodes [Result(bCanEnterTransition=Get bIsDBNO())]
 }
 Transition {
  path=AnimGraph.AnimGraphNode_StateMachine_0.Idle and Reactive.AnimStateTransitionNode_5.Transition
  nodes [Result(bCanEnterTransition=NOT Boolean(A=Get bIsDBNO()))]
 }
 EventGraph {nodes [
   Event BlueprintUpdateAnimation()
   > n1 = Sequence!(execute) -> (then_0, then_1, then_2:Exec)
   > n2 = Cast To PlayerPawn_Athena(Object=TryGetPawnOwner())
   > Set HeadSkelMesh(HeadSkelMesh:Skeletal Mesh Component Object Reference=GetSkeletalMeshForPartType(self=@n2, PartType=Head)) -> (Output_Get:Skeletal Mesh Component Object Reference)
   n3 = Branch(execute=@n1.then_2, Condition=Get bIsDBNO())
   n4 = GetOwnerSkeletalMeshComponent()
   n5 = BreakRotator(InRot=BreakTransform(InTransform=GetDeltaTransformFromRefPose(self=@n4, BoneName=spine_05, BaseName=root)).Rotation)
   n6 = ClampAxis!(Angle:float=@n5.Yaw) -> (ReturnValue:float)
   > Set Gimbal Rot(Gimbal Rot:Rotator=MakeRotator(Roll=float * float(A=@n6, B=-1) {is=title K2Node_PromotableOperator})) -> (Output_Get:Rotator)
   n7 = MakeRotator(Roll=float * float(A=@n5.Yaw, B=-1), Pitch=float * float(A=@n5.Yaw, B=0.5))
   Set Gimbal Rot(execute=@n3.else, Gimbal Rot=@n7)
   > n8 = Branch(Condition=Get bIsSkydiveDiveMode())
   > Set Gimbal Rot(Gimbal Rot=@n7)
   Set ADSGimbal(execute=@n8.else, ADSGimbal:Rotator=MakeRotator(Roll=float * float(A=@n6, B=-1), Pitch=float * float(A=@n6, B=0.5))) -> (Output_Get:Rotator)
   n9 = Get bIsSkydiveDiveMode()
   > Set Alpha Gimbal(Alpha Gimbal:double="To Float (Boolean)"(InBool=NOT Boolean(A=@n9))) -> (Output_Get:double)
   n10 = GetOwnerSkeletalMeshComponent()
   n11 = BreakRotator(InRot=BreakTransform(InTransform=GetDeltaTransformFromRefPose(self=@n10, BoneName=spine_05, BaseName=spine_01)).Rotation)
   n12 = FInterpEaseInOut!(A:double=0, B:double=MapRangeClamped(Value=@n11.Yaw, InRangeA=0, InRangeB=-20, OutRangeB=10), Alpha:double=0, Exponent:double=0) -> (ReturnValue:double)
   Set BP Rotate(execute=@n1.then_1, BP Rotate:Rotator=MakeRotator(Roll=@n12, Pitch=@n12)) -> (Output_Get:Rotator)
   n13 = MapRangeClamped(Value=@n11.Yaw, InRangeA=0, InRangeB=-10, OutRangeB=-5)
   > Set BP Translate(BP Translate:Vector=MakeVector(Y=FInterpEaseInOut(A=0, B=@n13, Alpha=0, Exponent=0))) -> (Output_Get:Vector)
   Event OnCharacterCustomizationLoaded()
   > n14 = Branch(Condition=IsCosmeticObjectDisplayedInFrontendContext!(CosmeticObject:Object Reference=GetOwningActor()) -> (ReturnValue:bool))
   > Do Once(Reset=@n16.then)
   > n15 = Delay(Duration=4)
   Set PlayReactive(execute=[@n17.then, @n15.then], PlayReactive:bool=true) -> (Output_Get:bool)
   > PrintString!(execute, InString=Hello, bPrintToScreen, bPrintToLog, TextColor, Duration=5, Key) -> (then)
   n16 = Set PlayReactive(execute=@n14.else, PlayReactive=false)
   AnimNotify_FE_ReactiveAnim() {is=title K2Node_Event}
   > Set PlayReactive(PlayReactive=false)
   > n17 = Delay(Duration=12)
   Event BlueprintPostEvaluateAnimation() {is=title K2Node_Event}
   > Set PlayReactive(PlayReactive=false)
   > Set PlayReactive(PlayReactive=false)
   n18 = Event BlueprintProcessAnimInputEvent()
   > n19 = Sequence!(execute) -> (then_0, then_1)
   > Branch(Condition=OR Boolean(A=IsNamed(self=@n18.AnimInputEvent, InEventName=aie.ChestOpened), B=IsNamed(self=@n18.AnimInputEvent, InEventName=aie.EnemyPlayerEliminated)))
   > Set PlayReactive(PlayReactive=true)
   > RetriggerableDelay(Duration=0.2)
   > Set PlayReactive(PlayReactive=false)
   Branch(
    execute=@n19.then_1
    Condition=OR Boolean(
     A=IsNamed(self=@n18.AnimInputEvent, InEventName=aie.PlayerVitality.ShieldsDecreased)
     B=IsNamed(self=@n18.AnimInputEvent, InEventName=aie.PlayerVitality.HealthDecreased)
    )
   )
   > Set PlayDamageReactive(PlayDamageReactive:bool=true) -> (Output_Get:bool)
   > RetriggerableDelay(Duration=0.2)
   > Set PlayDamageReactive(PlayDamageReactive=false)
   NOT Boolean(A=Get bIsCrouching())
  ]}
]
variables [
 ADSGimbal {type=Rotator}
 Alpha Gimbal {type=number}
 BP Rotate {type=Rotator}
 BP Translate {type=Vector}
 Gimbal Rot {type=Rotator}
 HeadSkelMesh {type=SkeletalMeshComponent}
 PlayDamageReactive {type=boolean}
 PlayReactive {type=boolean}
]
